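; Season 16 quest prop for Oro's tribute, with added logic for paying tribute or stealing a weapon.
; Overlap, damage or action tags complete interactions through the ConsolidatedQuestComponent, while discovery and ambient collisions drive reveals, audio and FX.
; Visuals change per player controller, with a no-quest visual test mode.
; Taking the weapon spawns a rift; without tribute it calls AttackPlayer, except for the Temple skin, which can get quest credit.
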(event EventBeginPlay
  (bind _p_ambient (Variables|Default|GetP_Ambient))
  ;[1 1 1] Special Hackery for Visual Testing
  (if (Variables|Design|QuestSetup|GetbTestVisualsNoQuestRequired)
    (Components|Activation|SetAutoActivate _p_ambient false)
    (Components|Activation|Deactivate _p_ambient)
    (if (Variables|Design|QuestSetup|GetbRecheckingTriggersOnSuccess)
      (Utilities|Time|SetTimerbyEvent (delegate Custom|VisualTestingRecheckOverlapping) 0.5 true))))

(event Custom|VisualTestingRecheckOverlapping
  ;[1 1 1] Special Hackery for Visual Testing
  (CallFunction|UpdateOverlappingActors))

(event Custom|PlayerInInteractRange
  (params (Actor :type Actor%20Object%20Reference)))

(event Custom|PlayerOutOfInteractRange
  (params (Actor :type Actor%20Object%20Reference)))

(event Custom|OnPlayerAction
  (params (PlayerController :type Fort%20Player%20Controller%20Object%20Reference) (ActionTag :type Gameplay%20Tag%20Structure) (PlayerTags :type Gameplay%20Tag%20Container%20Structure))
  (bind _btestvisualsnoquestrequired (Variables|Design|QuestSetup|GetbTestVisualsNoQuestRequired))
  ;[1 1 1] Player Action Tag to Interact
  (if (Variables|Default|GetbListenToTagToInteract)
    ;;[1 1 1] Debugging - Display Tags
    (Development|PrintString (Utilities|String|Append "; Player Controller: " (Utilities|GetDisplayName PlayerController) "; Actor Tags: " (GameplayTags|GetDebugStringfromGameplayTagContainer PlayerTags)) _btestvisualsnoquestrequired _btestvisualsnoquestrequired "(R=0.000000,G=0.660000,B=1.000000,A=1.000000)" 3e+01)
    (Development|PrintString (Utilities|String|Append "[Received] Event Tag: " (GameplayTags|GetDebugStringfromGameplayTag ActionTag)) _btestvisualsnoquestrequired _btestvisualsnoquestrequired "(R=1.000000,G=0.084869,B=0.000000,A=1.000000)" 3e+01)
    (Networking|SwitchHasAuthority
      (:Authority
        (if (GameplayTags|HasAnyTags (GameplayTags|MakeGameplayTagContainerfromTag ActionTag) (Variables|Design|QuestSetup|InteractionSettings|GetDamage%20or%20Interact%20Tag))
          (CallFunction|CheckTagsandComplete PlayerController))))))

(event OnComponentBeginOverlap(SM_HiddenInteractionCollision)
  (params (OverlappedComponent :type Primitive%20Component%20Object%20Reference) (OtherActor :type Actor%20Object%20Reference) (OtherComp :type Primitive%20Component%20Object%20Reference) (OtherBodyIndex :type Integer) (bFromSweep :type Boolean) (SweepResult :type Hit%20Result%20Structure%20%28by%20ref%29))
  ;[1 1 1] Overlap to Change State - Listen For Tag, Vehicles etc.
  (bind _data (Utilities|FlowControl|MergeExecutionwithData OtherActor (stale Custom|UpdateOverlappingActors(SM_HiddenInteractionCollision) OtherActor))
    (:Then
      (Networking|SwitchHasAuthority
        (:Authority
          (if (Variables|Default|GetbOverlapToInteract)
            (if (Variables|Default|GetbListenToTagToInteract)
              (bind _asfort_player_pawn (Utilities|Casting|CastToFortPlayerPawn _data))
              (if (Quests|IsQuestObjectiveActive (Quests|GetQuestManager (Fortnite|GetFortPlayerControllerfromActor _asfort_player_pawn) "Invalid") (Variables|/Script/FortniteGame.ConsolidatedQuestComponent|GetObjectiveBackendName (Variables|BuildingGameplayActorPropQuest|GetConsolidatedQuestComponent)))
                _
                (elif (Variables|Design|QuestSetup|GetbTestVisualsNoQuestRequired)
                  _))
              (GameplayEffects|ApplyGameplayEffectToTarget :self (Ability|GetAbilitySystemComponent self) :GameplayEffectClass (Variables|Default|GetTagListenerEffect) :Target (Ability|GetAbilitySystemComponent _asfort_player_pawn) :Level 0.0)
              (CallFunction|PlayerinInteractRange _asfort_player_pawn)
              (else
                (switch Utilities|FlowControl|Switch|SwitchonE_Athena_GenericQuestDamageTypes (Variables|Design|QuestSetup|InteractionSettings|GetInteractionType)
                  (:NewEnumerator9
                    ;;[1 1 1] Overlap  to Interact
                    (bind _asfort_vehicle_interface (Utilities|Casting|CastToFortVehicleInterface _data))
                    (for _array_element (Vehicle|GetAllPassengers _asfort_vehicle_interface)
                      (CallFunction|CheckTagsandComplete _array_element)))
                  (:NewEnumerator18
                    (bind _asphysics_ball_master (Utilities|Casting|CastToPhysicsBall_Master _data))
                    (CallFunction|CheckTagsandComplete (Actor|GetOwner _asphysics_ball_master)))
                  (:NewEnumerator20
                    (bind _asfort_player_pawn_2 (Utilities|Casting|CastToFortPlayerPawn _data))
                    (CallFunction|CheckTagsandComplete _asfort_player_pawn_2)))))))))))

(event Custom|UpdateOverlappingActors(SM_HiddenInteractionCollision)
  (params (OtherActor :type Actor%20Object%20Reference))
  ;[1 1 1] Overlap to Change State - Listen For Tag, Vehicles etc.
  (bind _data (Utilities|FlowControl|MergeExecutionwithData (stale OnComponentBeginOverlap(SM_HiddenInteractionCollision) OtherActor) OtherActor)
    (:Then
      (Networking|SwitchHasAuthority
        (:Authority
          (if (Variables|Default|GetbOverlapToInteract)
            (if (Variables|Default|GetbListenToTagToInteract)
              (bind _asfort_player_pawn (Utilities|Casting|CastToFortPlayerPawn _data))
              (if (Quests|IsQuestObjectiveActive (Quests|GetQuestManager (Fortnite|GetFortPlayerControllerfromActor _asfort_player_pawn) "Invalid") (Variables|/Script/FortniteGame.ConsolidatedQuestComponent|GetObjectiveBackendName (Variables|BuildingGameplayActorPropQuest|GetConsolidatedQuestComponent)))
                _
                (elif (Variables|Design|QuestSetup|GetbTestVisualsNoQuestRequired)
                  _))
              (GameplayEffects|ApplyGameplayEffectToTarget :self (Ability|GetAbilitySystemComponent self) :GameplayEffectClass (Variables|Default|GetTagListenerEffect) :Target (Ability|GetAbilitySystemComponent _asfort_player_pawn) :Level 0.0)
              (CallFunction|PlayerinInteractRange _asfort_player_pawn)
              (else
                (switch Utilities|FlowControl|Switch|SwitchonE_Athena_GenericQuestDamageTypes (Variables|Design|QuestSetup|InteractionSettings|GetInteractionType)
                  (:NewEnumerator9
                    ;;[1 1 1] Overlap  to Interact
                    (bind _asfort_vehicle_interface (Utilities|Casting|CastToFortVehicleInterface _data))
                    (for _array_element (Vehicle|GetAllPassengers _asfort_vehicle_interface)
                      (CallFunction|CheckTagsandComplete _array_element)))
                  (:NewEnumerator18
                    (bind _asphysics_ball_master (Utilities|Casting|CastToPhysicsBall_Master _data))
                    (CallFunction|CheckTagsandComplete (Actor|GetOwner _asphysics_ball_master)))
                  (:NewEnumerator20
                    (bind _asfort_player_pawn_2 (Utilities|Casting|CastToFortPlayerPawn _data))
                    (CallFunction|CheckTagsandComplete _asfort_player_pawn_2)))))))))))

(event OnComponentEndOverlap(SM_HiddenInteractionCollision)
  (params (OverlappedComponent :type Primitive%20Component%20Object%20Reference) (OtherActor :type Actor%20Object%20Reference) (OtherComp :type Primitive%20Component%20Object%20Reference) (OtherBodyIndex :type Integer))
  ;[1 1 1] Overlap to Change State - Listen For Tag, Vehicles etc.
  (Networking|SwitchHasAuthority
    (:Authority
      (if (Variables|Default|GetbOverlapToInteract)
        (if (Variables|Default|GetbListenToTagToInteract)
          (bind _asfort_player_pawn (Utilities|Casting|CastToFortPlayerPawn OtherActor))
          (GameplayEffects|RemoveActiveGameplayEffectbySourceEffect (Ability|GetAbilitySystemComponent _asfort_player_pawn) (Variables|Default|GetTagListenerEffect) (Ability|GetAbilitySystemComponent self) -1)
          (CallFunction|PlayerOutOfInteractRange _asfort_player_pawn))))))

(event Custom|ChangeStateForPlayer
  (params (FPC :type Fort%20Player%20Controller%20Object%20Reference))
  (bind _postinteraction_mis_override (Variables|Design|PostInteraction|GetPostInteraction_MIS_Override))
  (bind _bshouldapplycosmeticoverrides (Variables|Design|QuestSetup|GameplayTags,EffectsandCosmetics|GetbShouldApplyCosmeticOverrides))
  (bind _sm_quest_postinteraction (Variables|Default|GetSM_Quest_PostInteraction))
  (bind _sm_quest_preinteraction (Variables|Default|GetSM_Quest_PreInteraction))
  (bind _storedoriginalpreinteractionmaterial (Variables|Default|GetStoredOriginalPreInteractionMaterial))
  ;[1 1 1] Prop State Change on Interact
  (CallFunction|QuestLog "change state for player")
  (Variables|Debug|SetbOnVisualTestingCooldown true)
  (Utilities|Map|Remove (Variables|BuildingGameplayActorPropQuest|GetFortPlayerControllerToCanInteract) FPC)
  (CallFunction|QuestLog "remove controller interactions")
  ;;[1 0 0] This needs to work on just the interacting player in splitscreen
  (Components|Activation|Deactivate (Variables|Default|GetP_Ambient))
  (if (< (Variables|Design|CosmeticEvent|GetCosmeticTimeShift) 0.0)
    (Components|Activation|Activate (Variables|Default|GetP_Cosmetic) false)
    (CallFunction|SpawnCosmeticActor)
    (CallFunction|QuestLog "top delay")
    (Utilities|FlowControl|Delay (- (Math|Float|Absolute(Float) (Variables|Design|CosmeticEvent|GetCosmeticTimeShift)) (Variables|Design|PostInteraction|GetStateChangeDelay)))
    (else
      (CallFunction|QuestLog "bottom delay")))
  (Utilities|FlowControl|Delay (Variables|Design|PostInteraction|GetStateChangeDelay))
  (CallFunction|QuestLog "delay finished")
  (if (Variables|Default|GetbDamageToInteract)
    (Collision|SetCollisionEnabled (Variables|Default|GetSM_HiddenInteractionCollision) "NoCollision")
    (else
      _))
  ;;[1 1 1] Hiding or Swapping?
  (if (Variables|Design|PostInteraction|GetbHideAfterInteract)
    (CallFunction|QuestLog "hide after interact")
    (AddComponent|NODE_AddStaticMeshComponent-0)
    (bind (_direction _scaledowntime) (Timeline|T_UnhideCurve PlayFromStart)
      (:Update
        (Transformation|SetWorldScale3D (Variables|Default|GetSM_Quest_PreInteraction) (Math|Vector|Lerp(Vector) (.scale (Variables|Design|PreInteraction|GetPreInteraction%20Mesh%20Transform)) "0.010000,0.010000,0.010000" _scaledowntime)))
      (:Finished
        (PlayerController|SetCompHiddenControllerOnly FPC (Variables|Default|GetSM_Quest_PreInteraction) true true)
        (CallFunction|QuestLog "mesh now hidden in game")
        (Development|PrintString "Entered Post Interaction Logic" true true "(R=0.263882,G=1.000000,B=0.000000,A=1.000000)")
        ;;[1 1 1] Post-Interaction Particles and Sound
        (CallFunction|QuestLog "Before playing reward feedback")
        (Class|AthenaQuestFunctionLibrary|AthenaPlayQuestRewardFeedback (Variables|Design|PostInteraction|SuccessFX|GetbPlayFXOnPlayer) (Variables|Design|PostInteraction|SuccessFX|GetSuccessSound) (Variables|Default|GetSuccessNiagara) 0 self (Utilities|Casting|CastToFortPlayerPawn (Pawn|GetControlledPawn FPC)))
        (Development|PrintString "PlayedFeedback" true true "(R=1.000000,G=0.000000,B=0.089822,A=1.000000)")
        (Networking|SwitchHasAuthority
          (:Authority
            ;;;[1 1 1] Applying a Gameplay Effect to Player
            (CallFunction|ManagePlayerGEs FPC)
            (Development|PrintString "Has Cosmetic Override:")
            (Development|PrintString (Utilities|String|ToString(Boolean) _bshouldapplycosmeticoverrides))
            ;;;[1 1 1] Applying Cosmetic Override to Player
            (if _bshouldapplycosmeticoverrides
              (CallFunction|CosmeticOverridestoPlayer FPC)))
          (:Remote
            _))
        (Components|Activation|Activate (Variables|Default|GetPostInteractionAmbientAudio) false)
        (if (>= (Variables|Design|CosmeticEvent|GetCosmeticTimeShift) 0.0)
          (Utilities|FlowControl|Delay (Variables|Design|CosmeticEvent|GetCosmeticTimeShift))
          (Components|Activation|Activate (Variables|Default|GetP_Cosmetic) false)
          (CallFunction|SpawnCosmeticActor)
          (else
            _))
        (CallFunction|ResetForVisualTesting FPC)))
    (else
      (CallFunction|QuestLog "don't hide after interact")
      (Utilities|IsValid (Variables|Design|PostInteraction|GetSM_PostInteraction)
        (:Is%20Valid
          (PlayerController|SetCompHiddenControllerOnly FPC (Variables|Default|GetSM_Quest_PreInteraction) true false)
          (PlayerController|SetCompHiddenControllerOnly FPC _sm_quest_postinteraction false false)
          (Transformation|SetRelativeTransform _sm_quest_postinteraction (Variables|Design|PostInteraction|GetPostInteraction%20Mesh%20Transform) false false)
          (CallFunction|QuestLog "set relative transform"))
        (:Is%20Not%20Valid
          _))
      (Utilities|IsValid (Variables|Design|PostInteraction|GetPostInteraction_MIS_Override)
        (:Is%20Valid
          (Utilities|IsValid (Variables|Design|PostInteraction|GetSM_PostInteraction)
            (:Is%20Valid
              (Rendering|Material|SetMaterial (Variables|Default|GetSM_Quest_PostInteraction) 0 _postinteraction_mis_override))
            (:Is%20Not%20Valid
              (Rendering|Material|SetMaterial _sm_quest_preinteraction 0 _postinteraction_mis_override))))
        (:Is%20Not%20Valid
          (Utilities|IsValid _storedoriginalpreinteractionmaterial
            (:Is%20Valid
              (Rendering|Material|SetMaterial _sm_quest_preinteraction 0 _storedoriginalpreinteractionmaterial))
            (:Is%20Not%20Valid
              _)))))))

(event EventOnDamageServer
  (params (Damage :type Float%20%28single-precision%29) (DamageTags :type Gameplay%20Tag%20Container%20Structure%20%28by%20ref%29) (Momentum :type Vector :default "0, 0, 0") (HitInfo :type Hit%20Result%20Structure%20%28by%20ref%29) (InstigatedBy :type Controller%20Object%20Reference) (DamageCauser :type Actor%20Object%20Reference) (EffectContext :type Gameplay%20Effect%20Context%20Handle%20Structure))
  ;[1 1 1] Damage to Change State
  (if (Variables|Default|GetbDamageToInteract)
    (if (GameplayTags|HasAnyTags DamageTags (Variables|Design|QuestSetup|InteractionSettings|GetDamage%20or%20Interact%20Tag))
      (bind _asfort_player_controller (Utilities|Casting|CastToFortPlayerController (Utilities|SelectObject InstigatedBy (Ability|EffectContext|GetEffectCauser EffectContext) (Utilities|IsValid InstigatedBy :!class /Script/Engine.KismetSystemLibrary))))
      (CallFunction|CheckTagsandComplete _asfort_player_controller))))

(event Custom|LocalVisualsTest
  (params (FPC :type Fort%20Player%20Controller%20Object%20Reference))
  ;[1 1 1] Prop State Change on Interact
  ;;[1 1 1] Visual Testing Only (replicated)
  (if (Utilities|Equal(Object) FPC (Fortnite|GetFirstLocalFortPlayerController true))
    (CallFunction|ChangeStateforPlayer FPC)))

(event OnCustomSetVisualstoPassed(ConsolidatedQuestComponent)
  (params (PlayerController :type Fort%20Player%20Controller%20Object%20Reference) (bFullyInitialized :type Boolean)))

(event OnCustomSetVisualstoFailed(ConsolidatedQuestComponent)
  (params (PlayerController :type Fort%20Player%20Controller%20Object%20Reference)))

(event Custom|Reset%20for%20Visual%20Testing
  (params (FPC :type Fort%20Player%20Controller%20Object%20Reference))
  (bind _preinteraction_mis_override (Variables|Design|PreInteraction|GetPreInteraction_MIS_Override))
  ;[1 1 1] Prop State Change on Interact
  ;;[1 1 1] Reset to pre-interact state for visual testing mode
  (if (Variables|Design|QuestSetup|GetbTestVisualsNoQuestRequired)
    (Utilities|FlowControl|Delay (Variables|Design|QuestSetup|GetTestingCooldown))
    (Variables|Default|SetbBlockLocalInteractOverride false)
    (if (Variables|Design|PostInteraction|GetbHideAfterInteract)
      (Transformation|SetWorldScale3D (Variables|Default|GetSM_Quest_PreInteraction) (.scale (Variables|Design|PreInteraction|GetPreInteraction%20Mesh%20Transform)))
      (else
        _))
    (Components|Activation|Deactivate (Variables|Default|GetPostInteractionAmbientAudio))
    (Components|Activation|Activate (Variables|Default|GetP_Ambient) false)
    (PlayerController|SetCompHiddenControllerOnly FPC (Variables|Default|GetSM_Quest_PreInteraction) false true)
    (PlayerController|SetCompHiddenControllerOnly FPC (Variables|Default|GetSM_Quest_PostInteraction) true false)
    (Utilities|IsValid _preinteraction_mis_override
      (:Is%20Valid
        (Rendering|Material|SetMaterial (Variables|Default|GetSM_Quest_PreInteraction) 0 _preinteraction_mis_override))
      (:Is%20Not%20Valid
        _))
    (if (Variables|Design|QuestSetup|GameplayTags,EffectsandCosmetics|GetbShouldApplyCosmeticOverrides)
      (CallFunction|CosmeticOverridestoPlayer FPC true)
      (else
        _))
    (if (Variables|Default|GetbDamageToInteract)
      (Collision|SetCollisionEnabled (Variables|Default|GetSM_HiddenInteractionCollision) "QueryOnly"))
    (Variables|Debug|SetbOnVisualTestingCooldown false)))

(event OnQuestReady(ConsolidatedQuestComponent)
  (params (PlayerController :type Fort%20Player%20Controller%20Object%20Reference))
  ;[1 1 1] OnQuestReady(QuestComponent)
  (if (or (Quests|IsQuestObjectiveActive (Quests|GetQuestManager PlayerController "Invalid") (Variables|/Script/FortniteGame.ConsolidatedQuestComponent|GetObjectiveBackendName (Variables|BuildingGameplayActorPropQuest|GetConsolidatedQuestComponent))) (Variables|Design|QuestSetup|GetbTestVisualsNoQuestRequired))
    (PlayerController|SetActorHiddenControllerOnly PlayerController self false)
    (PlayerController|SetCompHiddenControllerOnly PlayerController (Variables|Default|GetSM_Quest_PostInteraction) true false)
    (if (not (Variables|Design|GetbUseDiscoverCollisionToUnhide))
      (Utilities|Map|Add (Variables|BuildingGameplayActorPropQuest|GetFortPlayerControllerToCanInteract) PlayerController "true"))
    (else
      (PlayerController|SetActorHiddenControllerOnly PlayerController self true))))

(event OnQuestNotReady(ConsolidatedQuestComponent)
  (params (PlayerController :type Fort%20Player%20Controller%20Object%20Reference))
  ;[1 1 1] OnQuestNotReady(QuestComponent)
  (if (Variables|Design|QuestSetup|GetbTestVisualsNoQuestRequired)
    _
    (elif (Utilities|Array|ContainsItem (Variables|Default|GetFortPlayerControllersWhoInteracted) PlayerController)
      ;;[1 1 1] If so, don't hide the actors player interacted with
      (if (Quest|HasCompletedObjectivewithName (Quests|GetQuestManager PlayerController "Invalid") (Class|ConsolidatedQuestComponent|GetQuestItemDefinition (Variables|BuildingGameplayActorPropQuest|GetConsolidatedQuestComponent)) (Variables|/Script/FortniteGame.ConsolidatedQuestComponent|GetObjectiveBackendName (Variables|BuildingGameplayActorPropQuest|GetConsolidatedQuestComponent)) :!class /Script/FortniteGame.FortQuestManager)
        _
        (elif (Variables|Design|QuestSetup|GetbVisibleToAll)
          (PlayerController|SetCompHiddenControllerOnly PlayerController (Variables|Default|GetAudioAndFX) true true)
          (PlayerController|SetCompHiddenControllerOnly PlayerController (Variables|Default|GetSM_Quest_PostInteraction) true true)
          (else
            (PlayerController|SetActorHiddenControllerOnly PlayerController self true)
            (PlayerController|SetCompHiddenControllerOnly PlayerController (Variables|Default|GetSM_Quest_PostInteraction) true false)
            (Rendering|SetCastShadow (Variables|Default|GetSM_Quest_PreInteraction) false)
            (Utilities|Map|Remove (Variables|BuildingGameplayActorPropQuest|GetFortPlayerControllerToCanInteract) PlayerController))))
      (else
        _))))

(event OnQuestComponentSuccessEvent(ConsolidatedQuestComponent)
  (params (PlayerController :type Fort%20Player%20Controller%20Object%20Reference))
  (bind _consolidatedquestcomponent (Variables|BuildingGameplayActorPropQuest|GetConsolidatedQuestComponent))
  (bind _returnvalue (Class|ConsolidatedQuestComponent|GetQuestItemDefinition _consolidatedquestcomponent))
  (bind _returnvalue_2 (Quests|GetQuestManager PlayerController "Invalid"))
  (bind _returnvalue_3 (Quests|GetQuestwithDefinition _returnvalue_2 _returnvalue))
  ;[1 1 1] OnQuestComponentSuccessEvent
  (CallFunction|UpdateOverlappingActors)
  (Networking|SwitchHasAuthority
    (:Authority
      (Utilities|IsValid _returnvalue_3
        (:Is%20Valid
          (if (Quest|HasCompletedAllObjectives _returnvalue_3 false)
            (if (Variables|Design|QuestSetup|GameplayTags,EffectsandCosmetics|GetbShouldApplyCosmeticOverrides)
              (if (Variables|Design|QuestSetup|GameplayTags,EffectsandCosmetics|GetbResetOverridesOnCompletion)
                (CallFunction|CosmeticOverridestoPlayer PlayerController true)
                (Development|PrintString "REMOVED" (Variables|Design|QuestSetup|GetbTestVisualsNoQuestRequired) (Variables|Design|QuestSetup|GetbTestVisualsNoQuestRequired) "(R=0.000000,G=1.000000,B=0.073006,A=1.000000)")))
            (else
              (Development|PrintString "NOT COMPLETED")
              (Development|PrintString (Utilities|GetDisplayName _returnvalue)))))))))

(event Custom|UpdateOverlappingActors(Collision_DiscoverRadius)
  (params (Other%20Actor :type Actor%20Object%20Reference))
  (bind (_value _returnvalue) (Utilities|Map|Find (Variables|BuildingGameplayActorPropQuest|GetFortPlayerControllerToCanInteract) (Variables|BuildingGameplayActorPropQuest|GetFortPlayerController_TempRef)))
  (bind _fortplayercontroller_tempref (Variables|BuildingGameplayActorPropQuest|GetFortPlayerController_TempRef))
  (bind _returnvalue_2 (Quests|GetQuestManager _fortplayercontroller_tempref "Invalid"))
  ;[1 1 1] Discovery Collision Overlap
  (bind _data (Utilities|FlowControl|MergeExecutionwithData Other%20Actor (stale OnComponentBeginOverlap(Collision_DiscoverRadius) OtherActor))
    (:Then
      (bind _asfort_player_pawn_athena (Utilities|Casting|CastToFortPlayerPawnAthena _data))
      (Utilities|IsValid (Fortnite|GetFortPlayerControllerfromActor _asfort_player_pawn_athena)
        (:Is%20Valid
          (bind _returnvalue_3 (CallFunction|PlayerTagsCheck (Fortnite|GetFortPlayerControllerfromActor _asfort_player_pawn_athena)))
          (if _returnvalue_3
            (Variables|BuildingGameplayActorPropQuest|SetFortPlayerController_TempRef (Fortnite|GetFortPlayerControllerfromActor _asfort_player_pawn_athena))
            (Networking|SwitchHasAuthority
              (:Authority
                ;;[1 1 1] Server needs to track the interact or send stat event for complex custom for "discovery'
                (if (Variables|Design|GetbCompleteFromDiscovery)
                  (Quest|SendComplexCustomStatEvent (Quests|GetQuestManager (Variables|BuildingGameplayActorPropQuest|GetFortPlayerController_TempRef) "Invalid") (Variables|BuildingGameplayActorPropQuest|GetFortPlayerController_TempRef) "(GameplayTags=)" (CallFunction|GetQuestTargetTags) 1)
                  (CallFunction|BPFXTestOverride (Variables|BuildingGameplayActorPropQuest|GetFortPlayerController_TempRef))
                  (else
                    (bind (_playercontrollerout _completedquest) (Quest|HasPlayerCompletedQuest (Variables|BuildingGameplayActorPropQuest|GetFortPlayerController_TempRef) (Class|ConsolidatedQuestComponent|GetQuestItemDefinition (Variables|BuildingGameplayActorPropQuest|GetConsolidatedQuestComponent))))
                    (if (or (and (not _completedquest) (Quests|IsQuestObjectiveActive (Quests|GetQuestManager (Variables|BuildingGameplayActorPropQuest|GetFortPlayerController_TempRef) "Invalid") (Variables|/Script/FortniteGame.ConsolidatedQuestComponent|GetObjectiveBackendName (Variables|BuildingGameplayActorPropQuest|GetConsolidatedQuestComponent)))) (and (Variables|Design|QuestSetup|GetbTestVisualsNoQuestRequired) (not (Variables|Debug|GetbOnVisualTestingCooldown))))
                      (Utilities|Map|Add (Variables|BuildingGameplayActorPropQuest|GetFortPlayerControllerToCanInteract) (Variables|BuildingGameplayActorPropQuest|GetFortPlayerController_TempRef) "true")))))
              (:Remote
                (if (not (and _value _returnvalue))
                  (Utilities|IsValid _returnvalue_2
                    (:Is%20Valid
                      (bind (_playercontrollerout_2 _completedquest_2) (Quest|HasPlayerCompletedQuest (Variables|BuildingGameplayActorPropQuest|GetFortPlayerController_TempRef) (Class|ConsolidatedQuestComponent|GetQuestItemDefinition (Variables|BuildingGameplayActorPropQuest|GetConsolidatedQuestComponent))))
                      (if (or (and (Variables|Design|QuestSetup|GetbTestVisualsNoQuestRequired) (not (Variables|Debug|GetbOnVisualTestingCooldown))) (and (Quests|IsQuestObjectiveActive _returnvalue_2 (Variables|/Script/FortniteGame.ConsolidatedQuestComponent|GetObjectiveBackendName (Variables|BuildingGameplayActorPropQuest|GetConsolidatedQuestComponent))) (not _completedquest_2)))
                        (PlayerController|SetCompHiddenControllerOnly (Variables|BuildingGameplayActorPropQuest|GetFortPlayerController_TempRef) (Variables|Default|GetSM_Quest_PreInteraction) false true)
                        (bind (_direction _scaledowntime) (Timeline|T_HideCurve PlayFromStart)
                          (:Update
                            (Transformation|SetWorldScale3D (Variables|Default|GetSM_Quest_PreInteraction) (Math|Vector|Lerp(Vector) "0.010000,0.010000,0.010000" (.scale (Variables|Design|PreInteraction|GetPreInteraction%20Mesh%20Transform)) _scaledowntime)))
                          (:PlayUnhideSound
                            (if (Variables|Design|GetbCompleteFromDiscovery)
                              (Class|AthenaQuestFunctionLibrary|AthenaPlayQuestRewardFeedback (Variables|Design|PostInteraction|SuccessFX|GetbPlayFXOnPlayer) (Variables|Design|PostInteraction|SuccessFX|GetSuccessSound) (Variables|Default|GetSuccessNiagara) 0 self (Game|Player|GetPlayerPawn (Variables|BuildingGameplayActorPropQuest|GetFortPlayerController_TempRef)))
                              (else
                                ;;[1 1 1] Add can interact on client
                                (Utilities|Map|Add (Variables|BuildingGameplayActorPropQuest|GetFortPlayerControllerToCanInteract) (Variables|BuildingGameplayActorPropQuest|GetFortPlayerController_TempRef) "true")
                                (Components|Activation|Activate (Variables|Default|GetUnHideAudio) false)
                                (Components|Activation|Activate (Variables|Default|GetP_UnHide) true)
                                (Components|Activation|Activate (Variables|Default|GetTributeCallOutAudio) false)
                                (Components|Activation|Activate (Variables|Default|GetN_Ambient) true)))))))))))))))))

(event OnCustomUpdateObjective(ConsolidatedQuestComponent)
  (params (PlayerController :type Fort%20Player%20Controller%20Object%20Reference) (AchievedCount :type Integer) (RequiredCount :type Integer))
  ;[1 1 1] In Testing Mode The Cosmetic Event Was called already
  (if (Variables|Design|QuestSetup|GetbTestVisualsNoQuestRequired)
    (if (Variables|Default|GetIsTemple)
      _)
    (elif (Variables|Design|GetbCompleteFromDiscovery)
      _
      (else
        ;[1 1 1] do we depend on the server rep from the quest manager for having the quest completed or assume they did?
        ;
        ; connect this backup if we want the former
        (CallFunction|ChangeStateforPlayer PlayerController)
        (Utilities|Array|AddUnique (Variables|Default|GetFortPlayerControllersWhoInteracted) PlayerController)))))

(event OnComponentBeginOverlap(Collision_DiscoverRadius)
  (params (OverlappedComponent :type Primitive%20Component%20Object%20Reference) (OtherActor :type Actor%20Object%20Reference) (OtherComp :type Primitive%20Component%20Object%20Reference) (OtherBodyIndex :type Integer) (bFromSweep :type Boolean) (SweepResult :type Hit%20Result%20Structure%20%28by%20ref%29))
  (bind (_value _returnvalue) (Utilities|Map|Find (Variables|BuildingGameplayActorPropQuest|GetFortPlayerControllerToCanInteract) (Variables|BuildingGameplayActorPropQuest|GetFortPlayerController_TempRef)))
  (bind _fortplayercontroller_tempref (Variables|BuildingGameplayActorPropQuest|GetFortPlayerController_TempRef))
  (bind _returnvalue_2 (Quests|GetQuestManager _fortplayercontroller_tempref "Invalid"))
  ;[1 1 1] Discovery Collision Overlap
  (bind _data (Utilities|FlowControl|MergeExecutionwithData (stale Custom|UpdateOverlappingActors(Collision_DiscoverRadius) Other%20Actor) OtherActor)
    (:Then
      (bind _asfort_player_pawn_athena (Utilities|Casting|CastToFortPlayerPawnAthena _data))
      (Utilities|IsValid (Fortnite|GetFortPlayerControllerfromActor _asfort_player_pawn_athena)
        (:Is%20Valid
          (bind _returnvalue_3 (CallFunction|PlayerTagsCheck (Fortnite|GetFortPlayerControllerfromActor _asfort_player_pawn_athena)))
          (if _returnvalue_3
            (Variables|BuildingGameplayActorPropQuest|SetFortPlayerController_TempRef (Fortnite|GetFortPlayerControllerfromActor _asfort_player_pawn_athena))
            (Networking|SwitchHasAuthority
              (:Authority
                ;;[1 1 1] Server needs to track the interact or send stat event for complex custom for "discovery'
                (if (Variables|Design|GetbCompleteFromDiscovery)
                  (Quest|SendComplexCustomStatEvent (Quests|GetQuestManager (Variables|BuildingGameplayActorPropQuest|GetFortPlayerController_TempRef) "Invalid") (Variables|BuildingGameplayActorPropQuest|GetFortPlayerController_TempRef) "(GameplayTags=)" (CallFunction|GetQuestTargetTags) 1)
                  (CallFunction|BPFXTestOverride (Variables|BuildingGameplayActorPropQuest|GetFortPlayerController_TempRef))
                  (else
                    (bind (_playercontrollerout _completedquest) (Quest|HasPlayerCompletedQuest (Variables|BuildingGameplayActorPropQuest|GetFortPlayerController_TempRef) (Class|ConsolidatedQuestComponent|GetQuestItemDefinition (Variables|BuildingGameplayActorPropQuest|GetConsolidatedQuestComponent))))
                    (if (or (and (not _completedquest) (Quests|IsQuestObjectiveActive (Quests|GetQuestManager (Variables|BuildingGameplayActorPropQuest|GetFortPlayerController_TempRef) "Invalid") (Variables|/Script/FortniteGame.ConsolidatedQuestComponent|GetObjectiveBackendName (Variables|BuildingGameplayActorPropQuest|GetConsolidatedQuestComponent)))) (and (Variables|Design|QuestSetup|GetbTestVisualsNoQuestRequired) (not (Variables|Debug|GetbOnVisualTestingCooldown))))
                      (Utilities|Map|Add (Variables|BuildingGameplayActorPropQuest|GetFortPlayerControllerToCanInteract) (Variables|BuildingGameplayActorPropQuest|GetFortPlayerController_TempRef) "true")))))
              (:Remote
                (if (not (and _value _returnvalue))
                  (Utilities|IsValid _returnvalue_2
                    (:Is%20Valid
                      (bind (_playercontrollerout_2 _completedquest_2) (Quest|HasPlayerCompletedQuest (Variables|BuildingGameplayActorPropQuest|GetFortPlayerController_TempRef) (Class|ConsolidatedQuestComponent|GetQuestItemDefinition (Variables|BuildingGameplayActorPropQuest|GetConsolidatedQuestComponent))))
                      (if (or (and (Variables|Design|QuestSetup|GetbTestVisualsNoQuestRequired) (not (Variables|Debug|GetbOnVisualTestingCooldown))) (and (Quests|IsQuestObjectiveActive _returnvalue_2 (Variables|/Script/FortniteGame.ConsolidatedQuestComponent|GetObjectiveBackendName (Variables|BuildingGameplayActorPropQuest|GetConsolidatedQuestComponent))) (not _completedquest_2)))
                        (PlayerController|SetCompHiddenControllerOnly (Variables|BuildingGameplayActorPropQuest|GetFortPlayerController_TempRef) (Variables|Default|GetSM_Quest_PreInteraction) false true)
                        (Timeline|T_HideCurve PlayFromStart)))))))))))))

(event EventFxTestOverride
  (params (FortPC :type Fort%20Player%20Controller%20Object%20Reference))
  (CallFunction|ChangeStateforPlayer (stale Custom|BP_FXTestOverride Fort%20PC)))

(event Custom|BP_FXTestOverride
  (params (Fort%20PC :type Fort%20Player%20Controller%20Object%20Reference))
  (CallFunction|ChangeStateforPlayer Fort%20PC))

(event OnComponentBeginOverlap(Collision_AmbientAudioAndFXActivation)
  (params (OverlappedComponent :type Primitive%20Component%20Object%20Reference) (OtherActor :type Actor%20Object%20Reference) (OtherComp :type Primitive%20Component%20Object%20Reference) (OtherBodyIndex :type Integer) (bFromSweep :type Boolean) (SweepResult :type Hit%20Result%20Structure%20%28by%20ref%29))
  (bind _consolidatedquestcomponent (Variables|BuildingGameplayActorPropQuest|GetConsolidatedQuestComponent))
  (bind _objectivebackendname (Variables|/Script/FortniteGame.ConsolidatedQuestComponent|GetObjectiveBackendName _consolidatedquestcomponent))
  ;[1 1 1] Activate Ambient Audio & FX if in test state or obj active
  (bind _asfort_player_pawn_athena (Utilities|Casting|CastToFortPlayerPawnAthena OtherActor))
  (bind _outactors (Actor|GetAllActorsOfClass "/NPCLibrary/NPCs/GoldenSkeleton/GoldenSkeleton/Pawns/BP_PlayerPawn_GoldenSkeleton.BP_PlayerPawn_GoldenSkeleton_C"))
  (Utilities|IsValid (Utilities|Array|Get(acopy) _outactors)
    (:Is%20Valid
      (EventDispatchers|BindEventtoOnDamaged (Utilities|Array|Get(acopy) _outactors) (delegate Custom|OroDamaged)))
    (:Is%20Not%20Valid
      _))
  (Networking|SwitchHasAuthority
    (:Remote
      (bind (_playercontrollerout _completedquest) (Quest|HasPlayerCompletedQuest (Fortnite|GetFortPlayerControllerfromActor _asfort_player_pawn_athena) (Class|ConsolidatedQuestComponent|GetQuestItemDefinition _consolidatedquestcomponent) _objectivebackendname))
      (if (and (not (Variables|Design|GetbUseDiscoverCollisionToUnhide)) (or (Math|Boolean|ANDBoolean (not _completedquest) (Quests|IsQuestObjectiveActive (Quests|GetQuestManager (Fortnite|GetFortPlayerControllerfromActor _asfort_player_pawn_athena) "Invalid") _objectivebackendname) (not (Variables|Design|QuestSetup|GetbTestVisualsNoQuestRequired))) (and (Variables|Design|QuestSetup|GetbTestVisualsNoQuestRequired) (not (Variables|Debug|GetbOnVisualTestingCooldown)))))
        (Utilities|IsValid (Utilities|Array|Get(acopy) _outactors)
          (:Is%20Valid
            (Components|Activation|Activate (Variables|Default|GetTributeCallOutAudio) false))
          (:Is%20Not%20Valid
            _))
        (Components|Activation|Activate (Variables|Default|GetP_Ambient) true)))))

(event Custom|OroDamaged
  (params (DamagedActor :type Actor%20Object%20Reference) (Damage :type Float%20%28single-precision%29) (InstigatedBy :type Controller%20Object%20Reference) (DamageCauser :type Actor%20Object%20Reference) (HitLocation :type Vector) (FHitComponent :type Primitive%20Component%20Object%20Reference) (BoneName :type Name) (Momentum :type Vector))
  ;[1 1 1] Activate Ambient Audio & FX if in test state or obj active
  (Utilities|FlowControl|Delay 1.0))

(event OnComponentEndOverlap(Collision_AmbientAudioAndFXActivation)
  (params (OverlappedComponent :type Primitive%20Component%20Object%20Reference) (OtherActor :type Actor%20Object%20Reference) (OtherComp :type Primitive%20Component%20Object%20Reference) (OtherBodyIndex :type Integer))
  (bind _postinteractionambientaudio (Variables|Default|GetPostInteractionAmbientAudio))
  ;[1 1 1] Disabling Ambient Sound and VFX on moving away
  (bind _asfort_player_pawn_athena (Utilities|Casting|CastToFortPlayerPawnAthena OtherActor))
  (Networking|SwitchHasAuthority
    (:Remote
      (Audio|Components|Audio|FadeOut _postinteractionambientaudio 3.0 0.0 "Linear" :!class /Script/Engine.AudioComponent)
      (Utilities|FlowControl|Sequence
        (:then_0
          (Utilities|FlowControl|Delay 3.0)
          (Components|Activation|Deactivate _postinteractionambientaudio)
          (Components|Activation|Deactivate (Variables|Default|GetP_Ambient)))
        (:then_1
          (bind _tributecalloutaudio (Variables|Default|GetTributeCallOutAudio)
            (:then
              (if (Audio|Components|Audio|IsPlaying _tributecalloutaudio)
                (Audio|Components|Audio|FadeOut _tributecalloutaudio 3.0 0.0 "Linear" :!class /Script/Engine.AudioComponent)))
            (:else)))))))

(event Custom|StartInteractOverrideTimerFailsafe
  ;[1 1 1] The client won't get the event that the quest has been completed right away,
  ; so we have to fake that it was, and then check that it was indeed completed
  ; and if it wasn't, then make sure to turn back functionality to normal. This fixes FORT-332126.
  ;
  ; We set it to false regardless because... splitscreen
  (Utilities|FlowControl|Delay 3.0)
  (Variables|Default|SetbBlockLocalInteractOverride false))

(event OnComponentBeginOverlap(Sphere)
  (params (OverlappedComponent :type Primitive%20Component%20Object%20Reference) (OtherActor :type Actor%20Object%20Reference) (OtherComp :type Primitive%20Component%20Object%20Reference) (OtherBodyIndex :type Integer) (bFromSweep :type Boolean) (SweepResult :type Hit%20Result%20Structure%20%28by%20ref%29))
  ;[1 1 1] Logic for Tribute or steal
  (if (Variables|Default|GetWeaponReady)
    (Development|PrintString "Weapon Ready ")
    (else
      (bind _asfort_pickup_athena (Utilities|Casting|CastToFortPickupAthena OtherActor))
      (Interaction|BindEventtoOnPickup _asfort_pickup_athena (delegate Custom|WeaponTaken))
      (Variables|Default|SetWeaponReady true))))

(event Custom|WeaponTaken
  (params (SelfActor :type Fort%20Pickup%20Object%20Reference) (InteractingPawn :type Fort%20Pawn%20Object%20Reference))
  (bind _player (Variables|Default|GetPlayer))
  (bind _returnvalue (Quests|GetQuestManager _player "Invalid"))
  ;[1 1 1] Logic for Tribute or steal
  (bind _asfort_player_pawn_athena (Utilities|Casting|CastToFortPlayerPawnAthena InteractingPawn))
  (Variables|Default|SetAs%20Fort%20Player%20Pawn%20Athena _asfort_player_pawn_athena)
  (bind _output_get (Variables|Default|SetPlayer (Fortnite|GetFortPlayerControllerfromActor _asfort_player_pawn_athena)))
  (if (Utilities|String|EqualExactly(String) (Utilities|GetDisplayName (.Character (Utilities|Struct|BreakFortAthenaLoadout (Variables|/Script/FortniteGame.FortPlayerController|GetCosmeticLoadoutPC _output_get)))) "CID_A_040_Athena_Commando_F_Temple")
    (Development|PrintString (Utilities|GetDisplayName (.Character (Utilities|Struct|BreakFortAthenaLoadout (Variables|/Script/FortniteGame.FortPlayerController|GetCosmeticLoadoutPC _output_get)))) true true "(R=0.000000,G=0.660000,B=1.000000,A=1.000000)" 5.0)
    (Variables|Default|SetIsTemple true)
    (else
      _))
  (Variables|Default|SetDisableInteractOnPickup true)
  (if (Variables|Default|GetPaidTribute)
    _
    (elif (Variables|Default|GetIsTemple)
      (Development|PrintString "StoleGunAsLaura" true true "(R=0.000000,G=0.660000,B=1.000000,A=1.000000)" 5.0)
      (else
        (Development|PrintString "Agro Oro" true true "(R=0.000000,G=0.660000,B=1.000000,A=1.000000)" 5.0)
        (CallFunction|AttackPlayer))))
  (Game|SpawnActorfromClass :Class "/Game/Athena/Items/ForagedItems/Rift/BGA_RiftPortal_Athena.BGA_RiftPortal_Athena_C" :SpawnTransform (Transformation|GetWorldTransform (Variables|Default|GetArrow)) :CollisionHandlingOverride "AlwaysSpawn" :TransformScaleMethod "OverrideRootScale")
  (if (Variables|Default|GetPaidTribute)
    (Utilities|FlowControl|Delay 8.0)
    (Actor|DestroyActor)
    (elif (Variables|Default|GetIsTemple)
      (if (not (Quest|HasCompletedQuest _returnvalue "/BattlepassS16/Items/QuestItems/quest_s16_tribute_q01.quest_s16_tribute_q01" :!class /Script/FortniteGame.FortQuestManager))
        (Quest|SendComplexCustomStatEvent _returnvalue _player "(GameplayTags=)" (CallFunction|GetQuestTargetTags) 1))
      (bind _returnvalue_2 (GUID|GenerateGuidfromObject _player))
      (Components|Activation|Activate (Variables|Default|GetP_Cosmetic) false)
      (Class|FortPlayerPawn|SwapPawnCosmeticItemandVariant (Game|Player|GetPlayerPawn _player) _returnvalue_2 (Variables|Design|QuestSetup|GameplayTags,EffectsandCosmetics|GetPlayer%20Cosmetic%20Overrides) true "Active" false)
      (CallFunction|AttackPlayer))))

(event Custom|AttackPlayer
  ;[1 1 1] Logic for Tribute or steal
  (Utilities|FlowControl|DoOnce false
    (:Completed
      (bind _outactors (Actor|GetAllActorsOfClass "/NPCLibrary/NPCs/GoldenSkeleton/GoldenSkeleton/Pawns/BP_PlayerPawn_GoldenSkeleton.BP_PlayerPawn_GoldenSkeleton_C"))
      (Utilities|IsValid (Utilities|Array|Get(acopy) _outactors)
        (:Is%20Valid
          (Variables|/Game/Athena/AI/NPCs/Base/BP_PlayerPawn_NonParticipant.BP_PlayerPawn_NonParticipant_C|SetAlertLevel "Threatened" (Utilities|Array|Get(acopy) _outactors))
          (Affiliation|ChangeAttitudeBetweenActors (Utilities|Array|Get(acopy) _outactors) (Variables|Default|GetAs%20Fort%20Player%20Pawn%20Athena) "Hostile")
          (Game|Damage|ApplyDamage (Utilities|Array|Get(acopy) _outactors) 1.0 0 0 "/Script/Engine.DamageType")
          (Utilities|FlowControl|Delay 0.5)
          (Audio|Components|Audio|Play (Variables|Default|GetVO_StealReact) 0.0)
          (Utilities|FlowControl|Delay 8.0)
          (Actor|DestroyActor))))))

(fn BlueprintCanInteract
  (params (InteractingPawn :type Fort%20Pawn%20Object%20Reference) (InteractionBeingAttempted :type EInteractionBeingAttempted%20Enum :default "FirstInteraction") (InteractionType :type TInteractionType%20Enum :default "IT_NoInteraction"))
  (bind _returnvalue (GameplayTags|GetOwnedGameplayTags InteractingPawn :!class /Script/GameplayTags.GameplayTagAssetInterface))
  (bind _prohibited_player_tags (Variables|Design|QuestSetup|GameplayTags,EffectsandCosmetics|GetProhibited%20Player%20Tags))
  (bind _returnvalue_2 (GameplayTags|HasAnyTags _returnvalue _prohibited_player_tags true))
  (bind _returnvalue_3 (not _returnvalue_2))
  (bind _required_player_tags (Variables|Design|QuestSetup|GameplayTags,EffectsandCosmetics|GetRequired%20Player%20Tags))
  (bind _returnvalue_4 (GameplayTags|HasAllTags _returnvalue _required_player_tags true))
  (bind _bdeposittointeract (Variables|Default|GetbDepositToInteract))
  (bind _returnvalue_5 (not _bdeposittointeract))
  (bind _returnvalue_6 (and _returnvalue_3 _returnvalue_4))
  (bind _resourceitemdefinition (Variables|Default|GetResourceItemDefinition))
  (bind _depositamount (Variables|Design|QuestSetup|InteractionSettings|GetDepositAmount))
  (bind _consolidatedquestcomponent (Variables|BuildingGameplayActorPropQuest|GetConsolidatedQuestComponent))
  (if (Variables|Default|GetDisableInteractOnPickup)
    (return false)
    (else
      (bind _asfort_player_controller (Utilities|Casting|CastToFortPlayerController (Pawn|GetController InteractingPawn))
        (:then
          (LocalVariables|SetLocal_FortPlayerControllerAthena _asfort_player_controller)
          (bind (_playercontrollerout _completedquest) (Quest|HasPlayerCompletedQuest _asfort_player_controller (Class|ConsolidatedQuestComponent|GetQuestItemDefinition _consolidatedquestcomponent)))
          (if (not _completedquest)
            (Networking|SwitchHasAuthority
              (:Authority
                (return (Math|Boolean|ANDBoolean (or (Quests|IsQuestObjectiveActive (Quests|GetQuestManager _asfort_player_controller "Invalid") (Variables|/Script/FortniteGame.ConsolidatedQuestComponent|GetObjectiveBackendName _consolidatedquestcomponent)) (Variables|Design|QuestSetup|GetbTestVisualsNoQuestRequired)) _returnvalue_6 (or _returnvalue_5 (>= (Inventory|GetItemQuantityOnPlayer _asfort_player_controller _resourceitemdefinition) _depositamount)))))
              (:Remote
                (if (not (Variables|Default|GetbBlockLocalInteractOverride))
                  (return (or (Math|Boolean|ANDBoolean (Utilities|Map|Find (Variables|BuildingGameplayActorPropQuest|GetFortPlayerControllerToCanInteract) _asfort_player_controller) (.ReturnValue (Utilities|Map|Find (Variables|BuildingGameplayActorPropQuest|GetFortPlayerControllerToCanInteract) _asfort_player_controller)) _returnvalue_6 (or _returnvalue_5 (>= (Inventory|GetItemQuantityOnPlayer _asfort_player_controller _resourceitemdefinition) _depositamount))) (Math|Boolean|ANDBoolean (Variables|Design|QuestSetup|GetbTestVisualsNoQuestRequired) (not (Variables|Debug|GetbOnVisualTestingCooldown)) _returnvalue_6 (or _returnvalue_5 (>= (Inventory|GetItemQuantityOnPlayer _asfort_player_controller _resourceitemdefinition) _depositamount)))))))))
          (return false))
        (:CastFailed
          (return false))))))

(fn BlueprintGetFailedInteractionString
  (params (InteractingPawn :type Fort%20Pawn%20Object%20Reference) (InteractionBeingAttempted :type EInteractionBeingAttempted%20Enum :default "FirstInteraction"))
  (bind _returnvalue (Fortnite|GetFortPlayerControllerfromActor InteractingPawn))
  (if (Variables|Default|GetbDepositToInteract)
    (if (or (Quests|IsQuestObjectiveActive (Quests|GetQuestManager _returnvalue "Invalid") (Variables|/Script/FortniteGame.ConsolidatedQuestComponent|GetObjectiveBackendName (Variables|BuildingGameplayActorPropQuest|GetConsolidatedQuestComponent))) (and (Variables|Design|QuestSetup|GetbTestVisualsNoQuestRequired) (not (Variables|Debug|GetbOnVisualTestingCooldown))))
      (return (Utilities|Text|ToText(String) (Utilities|String|Append (Utilities|String|Append (Utilities|String|Append (Utilities|String|ToString(Text) (Variables|Default|GetFailedInteractionVerb)) (Utilities|String|ToString(Integer) (- (Variables|Design|QuestSetup|InteractionSettings|GetDepositAmount) (Inventory|GetItemQuantityOnPlayer _returnvalue (Variables|Default|GetResourceItemDefinition))))) (Utilities|String|ToString(Text) (Variables|Default|GetFailedInteractionVerb_more))) (Utilities|String|ToString(Text) (Variables|/Script/ItemizationCoreRuntime.ItemDefinitionBase|GetItemName (Variables|Default|GetResourceItemDefinition))))))
      (else
        _))
    (else
      _))
  (return ""))

(fn BlueprintGetInteractionString
  (params (InteractingPawn :type Fort%20Pawn%20Object%20Reference) (InteractionBeingAttempted :type EInteractionBeingAttempted%20Enum :default "FirstInteraction"))
  (bind _interaction_verb (Variables|Design|QuestSetup|InteractionSettings|GetInteraction%20Verb))
  (if (Variables|Default|GetbDepositToInteract)
    (return (Utilities|Text|ToText(String) (Utilities|String|Append (Utilities|String|Append (Utilities|String|Append (Utilities|String|Append (Utilities|String|ToString(Text) _interaction_verb) " ") (Utilities|String|ToString(Integer) (Variables|Design|QuestSetup|InteractionSettings|GetDepositAmount))) " ") (Utilities|String|ToString(Text) (Variables|/Script/ItemizationCoreRuntime.ItemDefinitionBase|GetItemName (Variables|Default|GetResourceItemDefinition))))))
    (else
      (return _interaction_verb))))

(fn BlueprintGetInteractionTime
  (params (InteractingPawn :type Fort%20Pawn%20Object%20Reference) (InteractionBeingAttempted :type EInteractionBeingAttempted%20Enum :default "FirstInteraction"))
  (Utilities|IsValid (Variables|Default|GetNonPlayerConversationComponent)
    (:Is%20Valid
      (return 0.0 true))
    (:Is%20Not%20Valid
      (return 0.0 false))))

(fn BlueprintOnInteract
  (params (InteractingPawn :type Fort%20Pawn%20Object%20Reference) (InteractionBeingAttempted :type EInteractionBeingAttempted%20Enum :default "FirstInteraction"))
  (bind _interactingpawn (Variables|GetInteractingPawn))
  (bind _returnvalue (Fortnite|GetFortPlayerControllerfromActor _interactingpawn))
  (bind _returnvalue_2 (Fortnite|GetFortPlayerControllerfromActor InteractingPawn))
  (bind _resourceitemdefinition (Variables|Default|GetResourceItemDefinition))
  (Utilities|IsValid (Variables|Default|GetNonPlayerConversationComponent)
    (:Is%20Valid
      (Utilities|IsValid _returnvalue_2
        (:Is%20Valid
          (if (Conversation|CanStartConversationWith (Variables|Default|GetNonPlayerConversationComponent) _returnvalue_2)
            (Conversation|StartConversation (Variables|Default|GetNonPlayerConversationComponent) (Utilities|Array|Get(acopy) (GameplayTags|BreakGameplayTagContainer (Variables|Design|QuestSetup|InteractionSettings|GetDamage%20or%20Interact%20Tag))) _returnvalue_2 self :!class /Script/FortniteGame.FortNonPlayerConversationParticipantComponent)
            (else
              _))
          (Development|PrintString "OnServerInteract")
          (if (Variables|Design|QuestSetup|GetbTestVisualsNoQuestRequired)
            _)
          ;[1 1 1] Deposit to Interact logic
          (if (Variables|Default|GetbDepositToInteract)
            (if (>= (Inventory|GetItemQuantityOnPlayer _returnvalue _resourceitemdefinition) (Variables|Design|QuestSetup|InteractionSettings|GetDepositAmount))
              (bind _returnvalue_3 (InventoryManagement|FindExistingItemforDefinition _returnvalue _resourceitemdefinition))
              (InventoryManagement|RemoveItem _returnvalue (Item|GetItemGuid _returnvalue_3) (Variables|Design|QuestSetup|InteractionSettings|GetDepositAmount))
              (Variables|Default|SetPaidTribute true)
              (Variables|Default|SetDisableInteractOnPickup true)
              (return))))))
    (:Is%20Not%20Valid
      _)))

(fn BlueprintOnLocalInteract
  (params (InteractingPawn :type Fort%20Player%20Pawn%20Object%20Reference) (InteractionBeingAttempted :type EInteractionBeingAttempted%20Enum :default "FirstInteraction"))
  (Development|PrintString "BlueprintOnLocalInteract")
  (Variables|Default|SetbBlockLocalInteractOverride true)
  (if (Variables|Design|QuestSetup|GetbTestVisualsNoQuestRequired)
    (CallFunction|ChangeStateforPlayer (Fortnite|GetFortPlayerControllerfromActor InteractingPawn))
    (return false)))

(fn CheckTagsAndComplete
  (params (Actor :type Actor%20Object%20Reference))
  (bind _returnvalue (Fortnite|GetFortPlayerControllerfromActor Actor))
  (bind _returnvalue_2 (Quests|GetQuestManager _returnvalue "Invalid"))
  (bind _returnvalue_3 (CallFunction|PlayerTagsCheck _returnvalue (Variables|Design|QuestSetup|GameplayTags,EffectsandCosmetics|GetRequired%20Player%20Tags) (Variables|Design|QuestSetup|GameplayTags,EffectsandCosmetics|GetProhibited%20Player%20Tags)))
  (if _returnvalue_3
    (if (Variables|Design|QuestSetup|GetbTestVisualsNoQuestRequired)
      (if (not (Variables|Debug|GetbOnVisualTestingCooldown))
        (CallFunction|LocalVisualsTest _returnvalue)
        (CallFunction|ChangeStateforPlayer _returnvalue))
      (elif (Quests|IsQuestObjectiveActive _returnvalue_2 (Variables|/Script/FortniteGame.ConsolidatedQuestComponent|GetObjectiveBackendName (Variables|BuildingGameplayActorPropQuest|GetConsolidatedQuestComponent)))
        (Quest|SendComplexCustomStatEvent _returnvalue_2 _returnvalue "(GameplayTags=)" (CallFunction|GetQuestTargetTags) 1)))))

(fn CosmeticOverridestoPlayer
  (params (FPC :type Fort%20Player%20Controller%20Object%20Reference) (Remove :type Boolean)))

(fn ManagePlayerGEs
  (params (FPC :type Fort%20Player%20Controller%20Object%20Reference))
  (bind _returnvalue (Game|Player|GetPlayerPawn FPC))
  (bind _interaction_ges (Variables|Design|QuestSetup|GameplayTags,EffectsandCosmetics|GetInteraction%20GEs))
  (CallFunction|QuestLog "Beginning Manage Player GEs")
  (if (Variables|Design|QuestSetup|GameplayTags,EffectsandCosmetics|GetbShouldApplyGEs)
    (CallFunction|QuestLog "Apply GEs")
    (CallFunction|QuestLog (Utilities|String|ToString(Integer) (Utilities|Array|Length _interaction_ges)))
    (for _array_element _interaction_ges
      (CallFunction|QuestLog (Utilities|GetClassDisplayName _array_element))
      (Environment|ApplyGlobalEnvironmentGameplayEffecttoActor _returnvalue _array_element))
    (else
      _))
  (if (Variables|Design|QuestSetup|GameplayTags,EffectsandCosmetics|GetbShouldRemoveGEs)
    (CallFunction|QuestLog "Remove GEs")
    (for _array_element_2 (Variables|Design|QuestSetup|GameplayTags,EffectsandCosmetics|GetRemove%20GEs)
      (CallFunction|QuestLog (Utilities|GetClassDisplayName _array_element_2))
      (GameplayEffects|RemoveActiveGameplayEffectbySourceEffect (Ability|GetAbilitySystemComponent _returnvalue) _array_element_2 0 -1))
    (else
      _))
  (return))

(fn OnRep_DisableInteractOnPickup
  (Collision|SetCollisionResponsetoAllChannels (Variables|Default|GetSM_HiddenInteractionCollision) "ECR_Ignore"))

(fn OnRep_PaidTribute)

(fn OnRep_PlayAlterStingerSuccess
  (CallFunction|PlayAltarStinger "/NPCLibrary/NPCs/GoldenSkeleton/GoldenSkeleton/Audio/Music/Music_GoldenSkeleton_Stinger_Success_Cue.Music_GoldenSkeleton_Stinger_Success_Cue"))

(fn OnRep_PlayAlterStingerTheft
  (CallFunction|PlayAltarStinger "/NPCLibrary/NPCs/GoldenSkeleton/GoldenSkeleton/Audio/Music/Music_GoldenSkeleton_Stinger_Failure_Cue.Music_GoldenSkeleton_Stinger_Failure_Cue"))

(fn PlayAltarStinger
  (params (Sound :type Sound%20Base%20Object%20Reference))
  (bind _as_fort_player_pawn_athena (Variables|Default|GetAs%20Fort%20Player%20Pawn%20Athena)
    (:then
      (if (Pawn|IsLocallyControlled _as_fort_player_pawn_athena)
        (Audio|PlaySound2D Sound)))
    (:else)))

(fn ShouldDie
  (params (Damage :type Float%20%28single-precision%29) (EventInstigator :type Controller%20Object%20Reference) (DamageCauser :type Actor%20Object%20Reference) (bIsTest :type Boolean) (EffectContext :type Gameplay%20Effect%20Context%20Handle%20Structure%20%28by%20ref%29))
  (return false))

(fn SpawnCosmeticActor
  (Networking|SwitchHasAuthority
    (:Remote
      (if (Utilities|IsValidClass (Variables|Design|CosmeticEvent|GetClientCosmeticActorClass))
        (Utilities|IsValid (Variables|Default|GetClientCosmeticActorRef)
          (:Is%20Valid
            (if (Variables|Design|QuestSetup|GetbTestVisualsNoQuestRequired)
              _))
          (:Is%20Not%20Valid
            _)) ; this prevents more than 1 per player per actor
        (bind _returnvalue (Game|SpawnActorfromClass (Variables|Design|CosmeticEvent|GetClientCosmeticActorClass) (Math|Transform|MakeTransform (Transformation|GetActorLocation self) (Transformation|GetActorRotation self)) "AlwaysSpawn" "SelectDefaultAtRuntime" 0 0))
        (Variables|Default|SetClientCosmeticActorRef _returnvalue)
        (Development|PrintString "Spawned!\\")))))

(fn UpdateOverlappingActors
  (bind _sm_hiddeninteractioncollision (Variables|Default|GetSM_HiddenInteractionCollision))
  (bind _overlappingactors (Collision|GetOverlappingActors _sm_hiddeninteractioncollision "/Script/FortniteGame.FortPlayerPawn" :!class /Script/Engine.PrimitiveComponent))
  (bind _collision_discoverradius (Variables|Default|GetCollision_DiscoverRadius))
  (bind _overlappingactors_2 (Collision|GetOverlappingActors _collision_discoverradius "/Script/FortniteGame.FortPlayerPawn" :!class /Script/Engine.PrimitiveComponent))
  (if (Variables|Design|QuestSetup|GetbRecheckingTriggersOnSuccess)
    (if (Variables|Design|GetbUseDiscoverCollisionToUnhide)
      (if (Utilities|IsArrayNonEmpty _overlappingactors_2)
        (for _array_element _overlappingactors_2
          (CallFunction|UpdateOverlappingActors(CollisionDiscoverRadius) _array_element)))
      (else
        _))
    (if (Utilities|IsArrayNonEmpty _overlappingactors)
      (for _array_element_2 _overlappingactors
        (CallFunction|UpdateOverlappingActors(SMHiddenInteractionCollision) _array_element_2)))
    (return)
    (else
      (return))))

(fn ConstructionScript
  (bind (_origin _boxextent) (Collision|GetComponentBounds (Variables|Default|GetSM_HiddenInteractionCollision)))
  (bind (_origin_2 _boxextent_2) (Collision|GetComponentBounds (Variables|Default|GetSM_Quest_PreInteraction)))
  (bind _ambientsound (Variables|Design|PreInteraction|AmbientFX|GetAmbientSound))
  (bind _sm_hiddeninteractioncollision (Variables|Default|GetSM_HiddenInteractionCollision))
  (bind (_relativelocation_x _relativelocation_y) (Math|Vector|BreakVector (Variables|/Script/Engine.SceneComponent|GetRelativeLocation_X (Variables|Default|GetText_BackendNameHelper))))
  (bind _sm_hiddeninteractioncollision_2 (Variables|Default|GetSM_HiddenInteractionCollision))
  (bind _sm_hiddeninteractioncollision_3 (Variables|Default|GetSM_HiddenInteractionCollision))
  (bind _sm_hiddeninteractioncollision_4 (Variables|Default|GetSM_HiddenInteractionCollision))
  (bind _p_postinteractionambient (Variables|Default|GetP_PostInteractionAmbient))
  (bind _p_ambient (Variables|Default|GetP_Ambient))
  (bind _preinteraction_mis_override (Variables|Design|PreInteraction|GetPreInteraction_MIS_Override))
  (bind _postinteractionambientsound (Variables|Design|PostInteraction|AmbientFX|GetPostInteractionAmbientSound))
  (bind _cosmeticsound (Variables|Design|CosmeticEvent|GetCosmeticSound))
  (bind _sm_hiddeninteractioncollision_5 (Variables|Default|GetSM_HiddenInteractionCollision))
  (bind _unhide_particle (Variables|Design|PreInteraction|UnhideFX|GetUnhide%20Particle))
  (bind _unhidesound (Variables|Design|PreInteraction|UnhideFX|GetUnhideSound))
  (bind _postinteraction_mis_override (Variables|Design|PostInteraction|GetPostInteraction_MIS_Override))
  (bind _ambientniagarasystem (Variables|Design|PreInteraction|AmbientFX|GetAmbientNiagaraSystem))
  (bind _unhideniagarasystem (Variables|Design|PreInteraction|UnhideFX|GetUnhideNiagaraSystem))
  (bind _postinteractionambientniagarasystem (Variables|Design|PostInteraction|AmbientFX|GetPostInteractionAmbientNiagaraSystem))
  (bind _ambientparticle (Variables|Design|PreInteraction|AmbientFX|GetAmbientParticle))
  (bind _postinteractionambientparticle (Variables|Design|PostInteraction|AmbientFX|GetPostInteractionAmbientParticle))
  (bind _cosmeticparticle (Variables|Design|CosmeticEvent|GetCosmeticParticle))
  (bind _cosmeticniagarasystem (Variables|Design|CosmeticEvent|GetCosmeticNiagaraSystem))
  (bind _sm_hiddeninteractioncollision_6 (Variables|Default|GetSM_HiddenInteractionCollision))
  (Utilities|FlowControl|Sequence
    (:then_0
      ;[1 1 1] Master Quest Type Init
      (Variables|Design|QuestSetup|GameplayTags,EffectsandCosmetics|SetbShouldApplyGEs (Utilities|IsArrayNonEmpty (Variables|Design|QuestSetup|GameplayTags,EffectsandCosmetics|GetInteraction%20GEs)))
      (Variables|Design|QuestSetup|GameplayTags,EffectsandCosmetics|SetbShouldRemoveGEs (Utilities|IsArrayNonEmpty (Variables|Design|QuestSetup|GameplayTags,EffectsandCosmetics|GetRemove%20GEs)))
      (Variables|Design|QuestSetup|GameplayTags,EffectsandCosmetics|SetbShouldApplyCosmeticOverrides (Utilities|IsArrayNonEmpty (Variables|Design|QuestSetup|GameplayTags,EffectsandCosmetics|GetPlayer%20Cosmetic%20Overrides)))
      (switch Utilities|FlowControl|Switch|SwitchonE_Athena_GenericQuestActorTypes (Variables|Design|QuestSetup|GetbQuestActorType)
        (:NewEnumerator0
          (Variables|Design|SetbCompleteFromDiscovery false)
          (Variables|Design|SetbUseDiscoverCollisionToUnhide false))
        (:NewEnumerator1
          (Variables|Design|SetbCompleteFromDiscovery false)
          (Variables|Design|SetbUseDiscoverCollisionToUnhide true))
        (:NewEnumerator2
          (Variables|Design|SetbCompleteFromDiscovery true)
          (Variables|Design|SetbUseDiscoverCollisionToUnhide true))))
    (:then_1
      ;[1 1 1] Pre and Post Interaction Meshes Init
      (Components|StaticMesh|SetStaticMesh (Variables|Default|GetSM_Quest_PreInteraction) (Variables|Design|PreInteraction|GetSM_PreInteraction))
      (Components|StaticMesh|SetStaticMesh (Variables|Default|GetSM_Quest_PostInteraction) (Variables|Design|PostInteraction|GetSM_PostInteraction))
      (Transformation|SetRelativeTransform (Variables|Default|GetSM_Quest_PostInteraction) (Variables|Design|PostInteraction|GetPostInteraction%20Mesh%20Transform) false false)
      (Transformation|SetRelativeTransform (Variables|Default|GetSM_Quest_PreInteraction) (Variables|Design|PreInteraction|GetPreInteraction%20Mesh%20Transform) false false)
      ;;[1 1 1] Hide Or Not Post Interaction Mesh
      (if (Variables|Design|PostInteraction|GetbPreviewPostInteraction)
        (Utilities|IsValid (Variables|Design|PostInteraction|GetSM_PostInteraction)
          (:Is%20Valid
            (Transformation|SetWorldScale3D (Variables|Default|GetSM_Quest_PreInteraction) "0, 0, 0"))
          (:Is%20Not%20Valid
            _))
        (else
          (Transformation|SetWorldScale3D (Variables|Default|GetSM_Quest_PostInteraction) "0, 0, 0")))
      ;;[1 1 1] Change based on if we are using "discovery" or not
      (if (Variables|Design|GetbUseDiscoverCollisionToUnhide)
        (Transformation|SetWorldScale3D (Variables|Default|GetSM_Quest_PreInteraction) "0, 0, 0")
        (else
          (Components|DestroyComponent (Variables|Default|GetCollision_DiscoverRadius)))))
    (:then_2
      ;[1 1 1] Particle Init
      (Transformation|SetRelativeTransform (Variables|Default|GetP_Ambient) (Variables|Design|PreInteraction|AmbientFX|GetAmbient%20Particle%20Transform) false true)
      (Utilities|IsValid _ambientparticle
        (:Is%20Valid
          (Effects|Components|ParticleSystem|SetTemplate (Variables|Default|GetP_Ambient) _ambientparticle))
        (:Is%20Not%20Valid
          _))
      (Utilities|IsValid _ambientniagarasystem
        (:Is%20Valid
          (Niagara|SetNiagaraSystemAsset (Variables|Default|GetN_Ambient) _ambientniagarasystem true))
        (:Is%20Not%20Valid
          _))
      (Utilities|IsValid _postinteractionambientniagarasystem
        (:Is%20Valid
          (Niagara|SetNiagaraSystemAsset (Variables|Default|GetN_PostInteractionAmbient) _postinteractionambientniagarasystem true))
        (:Is%20Not%20Valid
          _))
      (Utilities|IsValid _postinteractionambientparticle
        (:Is%20Valid
          (Effects|Components|ParticleSystem|SetTemplate (Variables|Default|GetP_PostInteractionAmbient) _postinteractionambientparticle))
        (:Is%20Not%20Valid
          _))
      (Transformation|SetRelativeTransform (Variables|Default|GetP_PostInteractionAmbient) (Variables|Design|PostInteraction|AmbientFX|GetPostInteractionAmbientParticleTransform) false false)
      (if (Variables|Design|QuestSetup|GetbTestVisualsNoQuestRequired)
        (if (Variables|Design|PostInteraction|GetbPreviewPostInteraction)
          (Components|Activation|SetAutoActivate _p_ambient false)
          (Components|Activation|SetAutoActivate _p_postinteractionambient true)
          (else
            (Components|Activation|SetAutoActivate _p_ambient true)
            (Components|Activation|SetAutoActivate _p_postinteractionambient false)))
        (else
          _))
      (Utilities|IsValid _cosmeticparticle
        (:Is%20Valid
          (Effects|Components|ParticleSystem|SetTemplate (Variables|Default|GetP_Cosmetic) _cosmeticparticle))
        (:Is%20Not%20Valid
          _))
      (Utilities|IsValid _cosmeticniagarasystem
        (:Is%20Valid
          (Niagara|SetNiagaraSystemAsset (Variables|Default|GetN_Cosmetic) _cosmeticniagarasystem true))
        (:Is%20Not%20Valid
          _))
      (Transformation|SetRelativeTransform (Variables|Default|GetP_Cosmetic) (Variables|Design|CosmeticEvent|GetCosmeticParticleTransform) false false)
      (Utilities|IsValid _unhide_particle
        (:Is%20Valid
          (Effects|Components|ParticleSystem|SetTemplate (Variables|Default|GetP_UnHide) _unhide_particle))
        (:Is%20Not%20Valid
          _))
      (Utilities|IsValid _unhideniagarasystem
        (:Is%20Valid
          (Niagara|SetNiagaraSystemAsset (Variables|Default|GetN_UnHide) _unhideniagarasystem true))
        (:Is%20Not%20Valid
          _))
      (Transformation|SetRelativeTransform (Variables|Default|GetN_UnHide) (Variables|Design|PreInteraction|UnhideFX|GetUnhideParticleTransform) false false))
    (:then_3
      ;[1 1 1] Sound Init
      (Utilities|IsValid _ambientsound
        (:Is%20Valid
          (Audio|Components|Audio|SetSound (Variables|Default|GetTributeCallOutAudio) _ambientsound))
        (:Is%20Not%20Valid
          _))
      (Utilities|IsValid _postinteractionambientsound
        (:Is%20Valid
          (Audio|Components|Audio|SetSound (Variables|Default|GetPostInteractionAmbientAudio) _postinteractionambientsound))
        (:Is%20Not%20Valid
          _))
      (Utilities|IsValid _cosmeticsound
        (:Is%20Valid
          (Audio|Components|Audio|SetSound (Variables|Default|GetCosmeticAudio) _cosmeticsound))
        (:Is%20Not%20Valid
          _))
      (Utilities|IsValid _unhidesound
        (:Is%20Valid
          (Audio|Components|Audio|SetSound (Variables|Default|GetUnHideAudio) _unhidesound))))
    (:then_4
      ;[1 1 1] Radii Setup
      (Components|Sphere|SetSphereRadius (Variables|Default|GetCollision_AmbientAudioAndFXActivation) (Variables|Design|QuestSetup|GetAudioAndVFXHideRadius) false)
      (Components|Sphere|SetSphereRadius (Variables|Default|GetCollision_DiscoverRadius) (Variables|Design|QuestSetup|GetDiscoverRadius) false))
    (:then_6
      ;[1 1 1] Material Swaps
      (Utilities|IsValid _preinteraction_mis_override
        (:Is%20Valid
          (Variables|Default|SetStoredOriginalPreInteractionMaterial (Utilities|Array|Get(acopy) (Rendering|Material|GetMaterials (Variables|Default|GetSM_Quest_PreInteraction))))
          (if (not (Variables|Design|PostInteraction|GetbPreviewPostInteraction))
            ;;[1 1 1] If PreInteraction has swapped material
            (Rendering|Material|SetMaterial (Variables|Default|GetSM_Quest_PreInteraction) 0 _preinteraction_mis_override)))
        (:Is%20Not%20Valid
          (if (Variables|Design|PostInteraction|GetbPreviewPostInteraction)
            _)))
      ;;[1 1 1] If Previewing the Post Interaction State
      (Utilities|IsValid _postinteraction_mis_override
        (:Is%20Valid
          (Utilities|IsValid (Variables|Design|PostInteraction|GetSM_PostInteraction)
            (:Is%20Valid
              (Rendering|Material|SetMaterial (Variables|Default|GetSM_Quest_PostInteraction) 0 _postinteraction_mis_override))
            (:Is%20Not%20Valid
              (Rendering|Material|SetMaterial (Variables|Default|GetSM_Quest_PreInteraction) 0 _postinteraction_mis_override))))
        (:Is%20Not%20Valid
          (Rendering|Material|SetMaterial (Variables|Default|GetSM_Quest_PreInteraction) 0 (Variables|Default|GetStoredOriginalPreInteractionMaterial)))))
    (:then_7
      ;[1 1 1] Gross Damage to Interact Switcher
      ;;[1 1 1] Init
      (Collision|SetCollisionProfileName _sm_hiddeninteractioncollision_3 "FortTriggerOnlyInteractions" true :!class /Script/Engine.PrimitiveComponent)
      (Transformation|SetWorldScale3D _sm_hiddeninteractioncollision_3 (Variables|Design|QuestSetup|InteractionSettings|GetInteraction%20Collision%20Scale))
      (Variables|/Script/Engine.PrimitiveComponent|SetbGenerateOverlapEvents false _sm_hiddeninteractioncollision_3)
      (Variables|Default|SetbOverlapToInteract false)
      (Variables|Default|SetbDamageToInteract false)
      (Variables|Default|SetbDepositToInteract false)
      (Variables|Default|SetbListenToTagToInteract false)
      (Variables|Design|SetbConversationToComplete false)
      (CallFunction|QuestLog "In construction script")
      (switch Utilities|FlowControl|Switch|SwitchonE_Athena_GenericQuestDamageTypes (Variables|Design|QuestSetup|InteractionSettings|GetInteractionType)
        (:NewEnumerator0
          ;;[1 1 1] Tag Templates
          (Variables|Design|QuestSetup|InteractionSettings|SetDamage%20or%20Interact%20Tag "(GameplayTags=)"))
        (:NewEnumerator5
          (Variables|Design|QuestSetup|InteractionSettings|SetDamage%20or%20Interact%20Tag "(GameplayTags=((TagName=\"Weapon\")))"))
        (:NewEnumerator12
          (Variables|Design|QuestSetup|InteractionSettings|SetDamage%20or%20Interact%20Tag "(GameplayTags=((TagName=\"Weapon.Melee.Impact.Pickaxe\")))"))
        (:NewEnumerator6
          (Variables|Design|QuestSetup|InteractionSettings|SetDamage%20or%20Interact%20Tag "(GameplayTags=((TagName=\"Weapon.Melee\")))"))
        (:NewEnumerator7
          (Variables|Design|QuestSetup|InteractionSettings|SetDamage%20or%20Interact%20Tag "(GameplayTags=((TagName=\"Weapon.Ranged\")))"))
        (:NewEnumerator8
          (Variables|Design|QuestSetup|InteractionSettings|SetDamage%20or%20Interact%20Tag "(GameplayTags=((TagName=\"Weapon.Ranged.Grenade.Standard\"),(TagName=\"Weapon.Ranged.Heavy.C4\"),(TagName=\"Weapon.Ranged.Heavy.Grenade\"),(TagName=\"Weapon.Ranged.Heavy.Grenade_Launcher\"),(TagName=\"Weapon.Ranged.Heavy.Mortar\"),(TagName=\"Weapon.Ranged.Heavy.ProximityMine\"),(TagName=\"Weapon.Ranged.Heavy.Rocket_Launcher\")))")
          ;;[1 1 1] Collision Switching
          (Collision|SetCollisionProfileName _sm_hiddeninteractioncollision_4 "FortTriggerOnlyWeapons" true :!class /Script/Engine.PrimitiveComponent)
          ;;[1 1 1] Logic Mode Switching
          (Variables|/Script/Engine.PrimitiveComponent|SetbGenerateOverlapEvents true _sm_hiddeninteractioncollision_4)
          (Variables|Default|SetbDamageToInteract true))
        (:NewEnumerator9
          (Variables|Design|QuestSetup|InteractionSettings|SetDamage%20or%20Interact%20Tag "(GameplayTags=)")
          (Collision|SetCollisionProfileName _sm_hiddeninteractioncollision "FortTriggerOnlyPawns" true :!class /Script/Engine.PrimitiveComponent)
          (Collision|SetCollisionResponsetoAllChannels _sm_hiddeninteractioncollision "ECR_Ignore")
          (Collision|SetCollisionResponsetoChannel _sm_hiddeninteractioncollision "ECC_Vehicle" "ECR_Overlap")
          (Variables|/Script/Engine.PrimitiveComponent|SetbGenerateOverlapEvents true _sm_hiddeninteractioncollision)
          (Variables|Default|SetbOverlapToInteract true))
        (:NewEnumerator10
          (Variables|Design|QuestSetup|InteractionSettings|SetDamage%20or%20Interact%20Tag "(GameplayTags=)"))
        (:NewEnumerator11
          _)
        (:NewEnumerator13
          (Variables|Default|SetResourceItemDefinition "/Game/Items/ResourcePickups/Athena_WadsItemData.Athena_WadsItemData")
          (Variables|Design|QuestSetup|InteractionSettings|SetDamage%20or%20Interact%20Tag "(GameplayTags=)")
          (Variables|Default|SetbDepositToInteract true))
        (:NewEnumerator14
          (Variables|Default|SetResourceItemDefinition "/Game/Items/ResourcePickups/WoodItemData.WoodItemData"))
        (:NewEnumerator15
          (Variables|Default|SetResourceItemDefinition "/Game/Items/ResourcePickups/StoneItemData.StoneItemData"))
        (:NewEnumerator16
          (Variables|Default|SetResourceItemDefinition "/Game/Items/ResourcePickups/MetalItemData.MetalItemData"))
        (:NewEnumerator17
          (Collision|SetCollisionProfileName _sm_hiddeninteractioncollision_2 "FortTriggerOnlyPawns" true :!class /Script/Engine.PrimitiveComponent)
          (Collision|SetCollisionResponsetoAllChannels _sm_hiddeninteractioncollision_2 "ECR_Ignore")
          (Collision|SetCollisionResponsetoChannel _sm_hiddeninteractioncollision_2 "ECC_Pawn" "ECR_Overlap")
          (Variables|/Script/Engine.PrimitiveComponent|SetbGenerateOverlapEvents true _sm_hiddeninteractioncollision_2)
          (Variables|Default|SetbOverlapToInteract true)
          (Variables|Default|SetbListenToTagToInteract true))
        (:NewEnumerator18
          (Collision|SetCollisionProfileName _sm_hiddeninteractioncollision_6 "FortProjectileBall" true :!class /Script/Engine.PrimitiveComponent)
          (Collision|SetCollisionResponsetoAllChannels _sm_hiddeninteractioncollision_6 "ECR_Ignore")
          (Collision|SetCollisionResponsetoChannel _sm_hiddeninteractioncollision_6 "ECC_GameTraceChannel8" "ECR_Overlap")
          (Variables|/Script/Engine.PrimitiveComponent|SetbGenerateOverlapEvents true _sm_hiddeninteractioncollision_6)
          (Variables|Default|SetbOverlapToInteract true))
        (:NewEnumerator19
          (Variables|Design|SetbConversationToComplete true))
        (:NewEnumerator20
          (Collision|SetCollisionProfileName _sm_hiddeninteractioncollision_5 "FortTriggerOnlyPawns" true :!class /Script/Engine.PrimitiveComponent)
          (Collision|SetCollisionResponsetoAllChannels _sm_hiddeninteractioncollision_5 "ECR_Ignore")
          (Collision|SetCollisionResponsetoChannel _sm_hiddeninteractioncollision_5 "ECC_Pawn" "ECR_Overlap")
          (Variables|/Script/Engine.PrimitiveComponent|SetbGenerateOverlapEvents true _sm_hiddeninteractioncollision_5)
          (Variables|Default|SetbOverlapToInteract true)))
      (Collision|SetCollisionProfileName (Variables|Default|GetSM_HiddenInteractionCollision) "FortTriggerOnlyWeapons" true :!class /Script/Engine.PrimitiveComponent)
      (Variables|Default|SetbDamageToInteract true))
    (:then_8
      ;[1 1 1] Editor Only
      (Variables|Default|SetbErrorMessage false)
      (Variables|Default|SetDebugText "")
      (if (Variables|Design|QuestSetup|GetbTestVisualsNoQuestRequired)
        (Variables|Default|SetbErrorMessage true)
        (Variables|Default|SetDebugText (Utilities|String|Append (Variables|Default|GetDebugText) "
Visual Test Mode"))
        (else
          _))
      (if (Variables|Design|PostInteraction|GetbPreviewPostInteraction)
        (Variables|Default|SetbErrorMessage true)
        (Variables|Default|SetDebugText (Utilities|String|Append (Variables|Default|GetDebugText) "
Preview Post-Interaction"))
        (else
          _))
      (Utilities|IsValid (Class|ConsolidatedQuestComponent|GetQuestItemDefinition (Variables|BuildingGameplayActorPropQuest|GetConsolidatedQuestComponent))
        (:Is%20Valid
          _)
        (:Is%20Not%20Valid
          (Variables|Default|SetbErrorMessage true)
          (Variables|Default|SetDebugText (Utilities|String|Append (Variables|Default|GetDebugText) "
Need QuestItemDefinition"))))
      (if (not (>= (Utilities|String|Len (Utilities|String|ToString(Name) (Variables|/Script/FortniteGame.ConsolidatedQuestComponent|GetObjectiveBackendName (Variables|BuildingGameplayActorPropQuest|GetConsolidatedQuestComponent)))) 1))
        (Variables|Default|SetbErrorMessage true)
        (Variables|Default|SetDebugText (Utilities|String|Append (Variables|Default|GetDebugText) "
Need Objective Backend Name")))
      (if (and (Variables|Design|GetbUseDiscoverCollisionToUnhide) (> (Variables|Design|QuestSetup|GetAudioAndVFXHideRadius) (Variables|Design|QuestSetup|GetDiscoverRadius)))
        (Variables|Default|SetbErrorMessage true)
        (Variables|Default|SetDebugText (Utilities|String|Append (Variables|Default|GetDebugText) "
Discovery Radius Smaller than VFX radius"))
        (else
          _))
      (if (and (Variables|Default|GetbListenToTagToInteract) (not (Utilities|GameplayTagContainerNotEmpty (Variables|Design|QuestSetup|InteractionSettings|GetDamage%20or%20Interact%20Tag))))
        (Variables|Default|SetbErrorMessage true)
        (Variables|Default|SetDebugText (Utilities|String|Append (Variables|Default|GetDebugText) "
Need a Tag to listen to"))
        (else
          _))
      ;;[1 1 1] Positioning the Tooltip
      (Transformation|SetRelativeLocationAndRotation (Variables|Default|GetText_BackendNameHelper) (Math|Vector|MakeVector _relativelocation_x _relativelocation_y (* (Math|Float|Max(Float) (.z _boxextent_2) (.z _boxextent)) 2.0)) (Math|Rotator|InvertRotator (Transformation|GetActorRotation)) false false)
      (if (Variables|Default|GetbErrorMessage)
        (Rendering|Components|TextRender|SetTextRenderColor (Variables|Default|GetText_BackendNameHelper) (Variables|Default|GetErrorTextColor))
        (Rendering|Components|TextRender|SetText (Variables|Default|GetText_BackendNameHelper) (Utilities|Text|ToText(String) (Variables|Default|GetDebugText)))
        (else
          (Rendering|Components|TextRender|SetTextRenderColor (Variables|Default|GetText_BackendNameHelper) (Variables|Default|GetNormalTextColor))
          (Rendering|Components|TextRender|SetText (Variables|Default|GetText_BackendNameHelper) (Utilities|Text|ToText(String) (Utilities|String|Append "Obj: " (Utilities|String|ToString(Name) (Variables|/Script/FortniteGame.ConsolidatedQuestComponent|GetObjectiveBackendName (Variables|BuildingGameplayActorPropQuest|GetConsolidatedQuestComponent))) "
Tag: " (GameplayTags|GetDebugStringfromGameplayTagContainer (CallFunction|GetQuestTargetTags self))))))))
    (:then_9
      ;[1 1 1] Enable Conversation Component
      (if (not (Variables|Design|GetbConversationToComplete))
        (Components|DestroyComponent (Variables|Default|GetNonPlayerConversationComponent))))))
